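# Translucent lit material for Niagara mesh particles in the Whiff item's effects, giving a soft, pale yellow-green animated noise look.
# Opacity multiplies an inverted, boosted Fresnel by panning world-aligned noise projected on interpolated pre-skinned positions, then applies a depth fade.
# A second panning noise in object-relative world space blends two tints for adjusted emissive and also scales an upward vertex offset.
# A fauna-disguise gadget normal map is also sampled.
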Material /PrimalGameplay/Items/Whiff/FX/M_FXWhiffMesh
graph {
 properties {
  bAutomaticallySetUsageInEditor=false
  bUsedWithNiagaraMeshParticles=true
  blendMode=BLEND_Translucent
  pixelDepthOffsetMode=PDOM_Legacy
  translucencyLightingMode=TLM_VolumetricPerVertexDirectional
  usedShadingModels=Default Lit
 }
 nodes [
  n1 = ScalarParameter() {defaultValue=0.0238, parameterName=Opacity_FresExp}
  n2 = Multiply(A=OneMinus(None=Fresnel(ExponentIn=@n1)), B=ScalarParameter() {defaultValue=10, parameterName=Intensity Compoensation})
  n3 = TextureObject() {samplerType=SAMPLERTYPE_LinearGrayscale, texture=/Game/Effects/Fort_Effects/Textures/T_DarkAndLight_Noise}
  n4 = VectorParameter() {parameterName=Opacity Texture Size, defaultValue=<rgba 25 25 100 0>}
  n5 = VertexInterpolator(VS=PreSkinnedPosition())
  n6 = ComponentMask(None=VectorParameter() {parameterName=Opacity Panner Speed, defaultValue=<rgba 0 0.1 0 0>}.RGB) {g=true, r=true}
  DepthFade(
   Opacity=Multiply(
    A=Saturate(None=@n2)
    B=MaterialFunctionCall!(TextureObject=@n3, TextureSize=@n4.RGB, WorldPosition=@n5.PS, Export Float 4, In=@n6, World Space Normal, ProjectionTransitionContrast) {materialFunction=/Game/Athena/SafeZone/WorldAlignedTexture_ScalePan}.XYZ Texture
   ) {constA=1}
   FadeDistance=ScalarParameter() {defaultValue=1, parameterName=Fade Distance}
  )
  n7 = VectorParameter() {parameterName=Low, defaultValue=<rgba 0.8788 1 0.192 1>}
  n8 = VectorParameter() {parameterName=High, defaultValue=<rgba 0.9276 1 0.457 1>}
  n9 = MaterialFunctionCall(
   TextureObject=TextureObject() {samplerType=SAMPLERTYPE_LinearGrayscale, texture=&T_DarkAndLight_Noise}
   TextureSize=VectorParameter() {parameterName=WPO Texture Size, defaultValue=<rgba 100 100 100 0>}.RGB
   WorldPosition=Subtract(A=WorldPosition().XYZ, B=ObjectPositionWS())
   In=ComponentMask(None=VectorParameter() {parameterName=WPO Panner Speed, defaultValue=<rgba 0 0.1 0 0>}.RGB) {g=true, r=true}
  ) {materialFunction=&WorldAlignedTexture_ScalePan}
  MaterialFunctionCall(Emissive=Multiply(A=LinearInterpolate(A=@n7.RGB, B=@n8.RGB, Alpha=@n9.XYZ Texture), B=ScalarParameter() {defaultValue=0.25, parameterName=Emissive Intensity})) {materialFunction=&MF_FX_EmissiveAdjustments}
  Multiply(
   A=LinearInterpolate(B=Constant3Vector() {constant=<rgba 0 0 1 0>}, Alpha=@n9.XYZ Texture)
   B=ScalarParameter() {defaultValue=1, parameterName=VertexWPOAmount}
  )
  TextureSample() {samplerType=SAMPLERTYPE_Normal, texture=/PrimalGameplay/Items/Whiff/Textures/T_Gadget_Fauna_Disguise_N}
 ]
}
