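# Physics asset for the Whiff item mesh, a backpack rig with secondary-motion bodies for feathers, fur, a skull and a cape.
# A compound capsule-and-box body on backpack_root anchors tightly limited skull and fur bodies plus a five-link chain of cape boxes.
# Feather capsules hang with narrow limits from small master bodies (top and bottom on both sides, mid on left only), whose swing and twist are free relative to the root.
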PhysicsAsset /PrimalGameplay/Items/Whiff/Meshes/Whiff_Physics
bodies {
 Dyn_Bottom_Feather_l1 [{shapeType=Capsule, length=5, radius=0.5, center=<2.535 -0.1288 0.0892>, rotation=<rot 90 0 4.6364>}]
 Dyn_Bottom_Feather_l2 [{shapeType=Capsule, length=8, radius=0.5, center=<3.6875 -0.7017 0.3832>, rotation=<rot 82.6606 -179.9999 -175.5018>}]
 Dyn_Bottom_Feather_l3 [{shapeType=Capsule, length=10, radius=0.5, center=<5.0133 -0.6173 0.3394>, rotation=<rot 85.448 179.9999 -176.6905>}]
 Dyn_Bottom_Feather_l4 [{shapeType=Capsule, length=5, radius=0.5, center=<2.0649 -0.1138 0.0001>, rotation=<rot 90 0 0>}]
 Dyn_Bottom_Feather_r1 [{shapeType=Capsule, length=10, radius=0.75, center=<4.9869 -0.2 -0.486>, rotation=<rot 84.9612 -3.5182 -0.0001>}]
 Dyn_Bottom_Feather_r2 [{shapeType=Capsule, length=5, radius=0.5, center=<2.748 -0.3456 0>, rotation=<rot 90 0 0>}]
 Dyn_Bottom_Feather_r3 [{shapeType=Capsule, length=3, radius=0.25, center=<2.087 -0.043 0>, rotation=<rot 90 0 0>}]
 Dyn_Bottom_Feather_r4 [{shapeType=Capsule, length=4, radius=0.5, center=<0.5 0.3 0>, rotation=<rot 90 0 0>}]
 Dyn_Bottom_Master_l [{shapeType=Capsule, length=1, radius=1}]
 Dyn_Bottom_Master_r [{shapeType=Capsule, length=1, radius=1}]
 Dyn_Fur [{shapeType=Capsule, length=10, radius=5, center=<2 0 0>, rotation=<rot 0 -90 0>}]
 Dyn_Mid_Feather_l1 [{shapeType=Capsule, length=1.477, radius=0.7852, center=<1 -0.2933 -0.2268>, rotation=<rot 0 -91.7668 -82.6349>}]
 Dyn_Mid_Feather_l2 [{shapeType=Capsule, length=2.1631, radius=1.2433, center=<1 -0.5096 0.1873>, rotation=<rot 0 88.99 -66.356>}]
 Dyn_Mid_Feather_l3 [{shapeType=Capsule, length=1.0926, radius=0.4957, center=<1 -0.5003 0.0737>, rotation=<rot 0 72.4952 -73.4193>}]
 Dyn_Mid_Master_l [{shapeType=Capsule, length=1, radius=1}]
 Dyn_Skull [{shapeType=Capsule, length=2, radius=2, center=<1 0 0>, rotation=<rot 0 -90 0>}]
 Dyn_Top_Feather_l1 [{shapeType=Capsule, length=3.3919, radius=1.5706, center=<1 -0.1503 0.015>, rotation=<rot 0 89.4464 -86.0355>}]
 Dyn_Top_Feather_l2 [{shapeType=Capsule, length=2.7337, radius=1.4848, center=<1 -0.1318 0.4941>, rotation=<rot 0 88.0633 -82.4155>}]
 Dyn_Top_Feather_l3 [{shapeType=Capsule, length=2.1062, radius=1.1236, center=<1 -0.2983 0.2134>, rotation=<rot 0 -90.766 -86.2305>}]
 Dyn_Top_Feather_r1 [{shapeType=Capsule, length=3.3924, radius=1.5671, center=<1 0.2197 -0.2598>, rotation=<rot 0 -93.5358 -79.7502>}]
 Dyn_Top_Feather_r2 [{shapeType=Capsule, length=2.1842, radius=1.3798, center=<1 -0.2658 -0.5484>, rotation=<rot 0 -97.9293 -74.1161>}]
 Dyn_Top_Feather_r3 [{shapeType=Capsule, length=2.0512, radius=1.1115, center=<1 -0.2093 -0.2965>, rotation=<rot 0 -86.4346 -87.2703>}]
 Dyn_Top_Master_l [{shapeType=Capsule, length=1, radius=1}]
 Dyn_Top_Master_r [{shapeType=Capsule, length=1, radius=1}]
 backpack_root [
  {shapeType=Capsule, length=28.9, radius=16.8, center=<-0.0005 17.88 -29.26>, rotation=<rot 0 0 -11.24>}
  {shapeType=Capsule, length=14.1875, radius=17.375, center=<-0.0001 18.29 -55.88>, rotation=<rot 90 0 0>}
  {shapeType=Capsule, length=26.68, radius=15, center=<-0.0003 15.1 -22.017>, rotation=<rot -90 -0.014 0>}
  {shapeType=Box, extentX=1, extentY=1, extentZ=1}
  {shapeType=Box, extentX=106.4375, extentY=5.875, extentZ=122.125, center=<0.0003 15.5153 -74.1826>}
 ]
 dyn_cape01 [{shapeType=Box, extentX=8.125, extentY=1, extentZ=31.375, center=<4.22 0 0>}]
 dyn_cape02 [{shapeType=Box, extentX=8, extentY=1, extentZ=31.9375, center=<4.2056 -0.0001 0>}]
 dyn_cape03 [{shapeType=Box, extentX=7.8125, extentY=1, extentZ=33.9375, center=<4.188 0 0>}]
 dyn_cape04 [{shapeType=Box, extentX=22.25, extentY=1, extentZ=35.1875, center=<10.9811 -0.0001 0>}]
 dyn_cape05 [{shapeType=Box, extentX=14.3125, extentY=1, extentZ=35.25, center=<7.456 0.0001 0>}]
}
constraints [
 {bone1Name=Dyn_Skull, bone2Name=backpack_root, swing1LimitDegrees=1, swing2LimitDegrees=15, twistLimitDegrees=1}
 {bone1Name=Dyn_Fur, bone2Name=backpack_root, swing1LimitDegrees=10, swing2LimitDegrees=10, twistLimitDegrees=1}
 {bone1Name=dyn_cape01, bone2Name=backpack_root, swing1LimitDegrees=15, swing2LimitDegrees=15, twistLimitDegrees=5}
 {bone1Name=dyn_cape02, bone2Name=dyn_cape01, swing1LimitDegrees=20, swing2LimitDegrees=20, twistLimitDegrees=5}
 {bone1Name=dyn_cape03, bone2Name=dyn_cape02, swing1LimitDegrees=15, swing2LimitDegrees=15, twistLimitDegrees=5}
 {bone1Name=dyn_cape04, bone2Name=dyn_cape03, swing1LimitDegrees=10, swing2LimitDegrees=10, twistLimitDegrees=5}
 {bone1Name=dyn_cape05, bone2Name=dyn_cape04, swing1LimitDegrees=15, swing2LimitDegrees=15, twistLimitDegrees=5}
 {bone1Name=Dyn_Bottom_Feather_l1, bone2Name=Dyn_Bottom_Master_l, swing1LimitDegrees=30, swing2LimitDegrees=10, twistLimitDegrees=1}
 {bone1Name=Dyn_Bottom_Feather_l2, bone2Name=Dyn_Bottom_Master_l, swing1LimitDegrees=30, swing2LimitDegrees=10, twistLimitDegrees=1}
 {bone1Name=Dyn_Bottom_Feather_l3, bone2Name=Dyn_Bottom_Master_l, swing1LimitDegrees=30, swing2LimitDegrees=10, twistLimitDegrees=1}
 {bone1Name=Dyn_Bottom_Feather_l4, bone2Name=Dyn_Bottom_Master_l, swing1LimitDegrees=30, swing2LimitDegrees=10, twistLimitDegrees=1}
 {bone1Name=Dyn_Bottom_Feather_r1, bone2Name=Dyn_Bottom_Master_r, swing1LimitDegrees=30, swing2LimitDegrees=10, twistLimitDegrees=1}
 {bone1Name=Dyn_Bottom_Feather_r2, bone2Name=Dyn_Bottom_Master_r, swing1LimitDegrees=30, swing2LimitDegrees=10, twistLimitDegrees=1}
 {bone1Name=Dyn_Bottom_Feather_r3, bone2Name=Dyn_Bottom_Master_r, swing1LimitDegrees=30, swing2LimitDegrees=10, twistLimitDegrees=1}
 {bone1Name=Dyn_Bottom_Feather_r4, bone2Name=Dyn_Bottom_Master_r, swing1LimitDegrees=30, swing2LimitDegrees=1, twistLimitDegrees=1}
 {
  bone1Name=Dyn_Bottom_Master_l
  bone2Name=backpack_root
  swing1LimitDegrees=45
  swing1Motion=Free
  swing2LimitDegrees=45
  swing2Motion=Free
  twistLimitDegrees=45
  twistMotion=Free
 }
 {
  bone1Name=Dyn_Bottom_Master_r
  bone2Name=backpack_root
  swing1LimitDegrees=45
  swing1Motion=Free
  swing2LimitDegrees=45
  swing2Motion=Free
  twistLimitDegrees=45
  twistMotion=Free
 }
 {
  bone1Name=Dyn_Top_Master_l
  bone2Name=backpack_root
  swing1LimitDegrees=45
  swing1Motion=Free
  swing2LimitDegrees=45
  swing2Motion=Free
  twistLimitDegrees=45
  twistMotion=Free
 }
 {
  bone1Name=Dyn_Top_Master_r
  bone2Name=backpack_root
  swing1LimitDegrees=45
  swing1Motion=Free
  swing2LimitDegrees=45
  swing2Motion=Free
  twistLimitDegrees=45
  twistMotion=Free
 }
 {
  bone1Name=Dyn_Mid_Master_l
  bone2Name=backpack_root
  swing1LimitDegrees=45
  swing1Motion=Free
  swing2LimitDegrees=45
  swing2Motion=Free
  twistLimitDegrees=45
  twistMotion=Free
 }
 {bone1Name=Dyn_Top_Feather_l1, bone2Name=Dyn_Top_Master_l, swing1LimitDegrees=20, swing2LimitDegrees=10, twistLimitDegrees=1}
 {bone1Name=Dyn_Top_Feather_l2, bone2Name=Dyn_Top_Master_l, swing1LimitDegrees=20, swing2LimitDegrees=10, twistLimitDegrees=1}
 {bone1Name=Dyn_Top_Feather_l3, bone2Name=Dyn_Top_Master_l, swing1LimitDegrees=20, swing2LimitDegrees=10, twistLimitDegrees=1}
 {bone1Name=Dyn_Top_Feather_r1, bone2Name=Dyn_Top_Master_r, swing1LimitDegrees=20, swing2LimitDegrees=10, twistLimitDegrees=1}
 {bone1Name=Dyn_Top_Feather_r2, bone2Name=Dyn_Top_Master_r, swing1LimitDegrees=20, swing2LimitDegrees=10, twistLimitDegrees=1}
 {bone1Name=Dyn_Top_Feather_r3, bone2Name=Dyn_Top_Master_r, swing1LimitDegrees=20, swing2LimitDegrees=10, twistLimitDegrees=1}
 {bone1Name=Dyn_Mid_Feather_l1, bone2Name=Dyn_Mid_Master_l, swing1LimitDegrees=20, swing2LimitDegrees=10, twistLimitDegrees=1}
 {bone1Name=Dyn_Mid_Feather_l2, bone2Name=Dyn_Mid_Master_l, swing1LimitDegrees=20, swing2LimitDegrees=10, twistLimitDegrees=1}
 {bone1Name=Dyn_Mid_Feather_l3, bone2Name=Dyn_Mid_Master_l, swing1LimitDegrees=20, swing2LimitDegrees=10, twistLimitDegrees=1}
]
