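# Animation Blueprint for the Whiff item's skeleton, a backpack with cape bones that follows the attached parent's pose and adds physics-driven secondary motion.
# When bIsInTheAir is true, a RigidBody simulation takes gravity and component motion scales from variables.
# Otherwise, cape bones are offset before a second simulation with fixed settings and translation clamping.
# Both simulate in backpack_root space, and their blended result gets a final offset, roll and upscale on that bone.
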AnimBlueprint /PrimalGameplay/Items/Whiff/Meshes/Whiff_AnimBP
target_skeleton=&/Whiff_Skeleton
graphs [
 AnimGraph {nodes [
   "Transform (Modify) Bone - Bone: backpack_root"(Alpha=1, Translation="0,0,5", Rotation="0,0,5", Scale="1.5,1.5,1.5") {
    rotationMode=BMM_Additive
    scaleMode=BMM_Replace
    translationMode=BMM_Additive
    alphaBoolBlend {blendOption=Linear}
    boneToModify {boneName=backpack_root}
    rotation {roll=5}
    scale=<1.5 1.5 1.5>
    translation {z=5}
   }
   n1 = RigidBody!(
    ComponentPose=Local To Component(LocalPose=Copy Pose From Mesh() {bUseAttachedParent=true})
    Alpha
    OverrideWorldGravity:Vector=Get FinalGravity() -> (FinalGravity:Vector)
    ExternalForce
    ComponentLinearAccScale:Vector=Get CompLinearAccScale() -> (CompLinearAccScale:Vector)
    ComponentLinearVelScale:Vector=Get CompLinearVelScale() -> (CompLinearVelScale:Vector)
   ) -> (Pose) {
    bOverrideWorldGravity=true
    overridePhysicsAsset=&/Whiff_Physics
    simulationSpace=BaseBoneSpace
    alphaBoolBlend {blendOption=Linear}
    baseBoneRef {boneName=backpack_root}
    componentAppliedLinearAccClamp=<1500 1500 1500>
    componentLinearAccScale=<0.5 0.5 0.5>
    componentLinearVelScale=<0.4 0.4 0.4>
    overrideWorldGravity {z=-980}
   }
   n2 = Component To Local(ComponentPose=@n1)
   n3 = "Transform (Modify) Bone - Bone: dyn_cape01"(
    ComponentPose=Local To Component(LocalPose=Copy Pose From Mesh() {bUseAttachedParent=true})
    Alpha=1
    Translation="0,-5,0"
    Rotation="0,0,0"
    Scale="1,1,1"
   ) {
    rotationMode=BMM_Additive
    rotationSpace=BCS_BoneSpace
    translationMode=BMM_Additive
    translationSpace=BCS_BoneSpace
    alphaBoolBlend {blendOption=Linear}
    boneToModify {boneName=dyn_cape01}
    translation {y=-5}
   }
   n4 = "Transform (Modify) Bone - Bone: dyn_cape03"(ComponentPose=@n3, Alpha=1, Translation="0,-2,0", Rotation="0,0,0", Scale="1,1,1") {
    rotationMode=BMM_Additive
    rotationSpace=BCS_BoneSpace
    translationMode=BMM_Additive
    translationSpace=BCS_BoneSpace
    alphaBoolBlend {blendOption=Linear}
    boneToModify {boneName=dyn_cape03}
    translation {y=-2}
   }
   n5 = "Transform (Modify) Bone - Bone: dyn_cape05"(ComponentPose=@n4, Alpha=1, Translation="0,-2,0", Rotation="0,0,0", Scale="1,1,1") {
    rotationMode=BMM_Additive
    rotationSpace=BCS_BoneSpace
    translationMode=BMM_Additive
    translationSpace=BCS_BoneSpace
    alphaBoolBlend {blendOption=Linear}
    boneToModify {boneName=dyn_cape05}
    translation {y=-2}
   }
   n6 = RigidBody!(ComponentPose=@n5, Alpha, ExternalForce) -> (Pose) {
    bClampLinearTranslationLimitToRefPose=true
    bOverrideWorldGravity=true
    overridePhysicsAsset=&Whiff_Physics
    simulationSpace=BaseBoneSpace
    alphaBoolBlend {blendOption=Linear}
    baseBoneRef {boneName=backpack_root}
    componentAppliedLinearAccClamp=<1200 1200 1200>
    componentLinearAccScale=<0.5 0.5 0.5>
    componentLinearVelScale=<0.5 0.5 0.5>
    overrideWorldGravity {z=-980}
   }
   n7 = Component To Local(ComponentPose=@n6)
   n8 = Get bIsInTheAir() -> (bIsInTheAir:bool)
   n9 = "Transform (Modify) Bone - Bone: backpack_root"(
    ComponentPose=Local To Component(LocalPose=Blend Poses by bool(BlendPose_0=@n2, BlendPose_1=@n7, BlendTime_0=0.2, BlendTime_1=0.5, bActiveValue=@n8) {blendType=Linear})
    Alpha=1
    Translation="0,0,5"
    Rotation="0,0,5"
    Scale="1.5,1.5,1.5"
   ) {
    rotationMode=BMM_Additive
    scaleMode=BMM_Replace
    translationMode=BMM_Additive
    alphaBoolBlend {blendOption=Linear}
    boneToModify {boneName=backpack_root}
    rotation {roll=5}
    scale=<1.5 1.5 1.5>
    translation {z=5}
   }
   Output Pose(Result=Component To Local(ComponentPose=@n9)) {layerGroup=None}
  ]}
 EventGraph {nodes [Event BlueprintUpdateAnimation()]}
]
variables []
